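; Ridable component for the boar that handles impact feedback while ridden, apparently for sprint charges, and rider dismounts.
; It binds to the owner's collision reaction component on BeginPlay and unbinds on EndPlay.
; With a valid rider, hits branch by building, pawn or vehicle to apply configured effects and rider rumble/camera-shake cues, plus an impact VFX/SFX cue on the boar.
; Dismounting sends a StoppedRiding gameplay event and applies a loot-discovery cooldown to the boar.
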(event EventBeginPlay
  (bind _returnvalue (Components|GetOwner))
  (bind _returnvalue_2 (Actor|GetComponentbyClass _returnvalue "/Script/FortniteGame.FortAthenaCollisionReactionComponent"))
  (|CallParentFunction)
  (Utilities|IsValid _returnvalue_2
    (:Is%20Valid
      (Interaction|BindEventtoOnReaction _returnvalue_2 (delegate Custom|OnReactionEvent)))))

(event Custom|OnReactionEvent
  (params (HitActor :type Actor%20Object%20Reference) (bFromAsyncSweepBox :type Boolean) (CollisionReactionType :type FCollisionReactionType%20Enum :default "Nothing") (bIsFirstContinuousReactionOnDelayableActor :type Boolean) (HitResult :type Hit%20Result%20Structure%20%28by%20ref%29))
  (CallFunction|OnReaction HitActor HitResult))

(event EventEndPlay
  (params (EndPlayReason :type EEndPlayReason%20Enum :default "Destroyed"))
  (bind _returnvalue (Components|GetOwner))
  (bind _returnvalue_2 (Actor|GetComponentbyClass _returnvalue "/Script/FortniteGame.FortAthenaCollisionReactionComponent"))
  (|CallParentFunction EndPlayReason)
  (Utilities|IsValid _returnvalue_2
    (:Is%20Valid
      (Interaction|UnbindallEventsfromOnReaction _returnvalue_2))))

(event EventTick
  (params (DeltaSeconds :type Float%20%28single-precision%29)))

(event EventHandleRiderStartedRiding
  (params (Rider :type Rider%20Component%20Object%20Reference) (SlotIndex :type Integer))
  (|CallParentFunction Rider 0))

(event EventHandleRiderStoppedRiding
  (params (Rider :type Rider%20Component%20Object%20Reference) (SlotIndex :type Integer))
  (bind _returnvalue (Components|GetOwner))
  (bind _returnvalue_2 (Ability|GetAbilitySystemComponent _returnvalue))
  (Ability|SendGameplayEventtoActor (Components|GetOwner) "(TagName=\"Gameplay.Action.Player.StoppedRiding\")" (Utilities|Struct|MakeGameplayEventData "(TagName=\"\")"))
  (Utilities|IsValid _returnvalue
    (:Is%20Valid
      (Utilities|IsValid _returnvalue_2
        (:Is%20Valid
          (GameplayEffects|ApplyGameplayEffectToSelf :self _returnvalue_2 :GameplayEffectClass "/Irwin/AI/Prey/Burt/Abilities/GE_NPC_Irwn_Prey_Burt_DiscoverLootCooldown.GE_NPC_Irwn_Prey_Burt_DiscoverLootCooldown_C" :Level 0.0))
        (:Is%20Not%20Valid
          _))
      (|CallParentFunction Rider 0))
    (:Is%20Not%20Valid
      _)))

(event EventHandleAbilityStarted
  (|CallParentFunction))

(fn OnReaction
  (params (Object :type Object%20Reference) (HitResult :type Hit%20Result%20Structure))
  (bind _donotuse_primaryrider (Variables|RidableComponent|GetDoNotUse_PrimaryRider))
  (bind _returnvalue (Components|GetOwner _donotuse_primaryrider))
  (bind _gesprintimpactpawn (Variables|Default|GetGESprintImpactPawn))
  (bind _gesprintimpactvehicle (Variables|Default|GetGESprintImpactVehicle))
  ;[1 1 1] Get Rider Actgor
  (Utilities|IsValid _donotuse_primaryrider
    (:Is%20Valid
      (Utilities|IsValid _returnvalue
        (:Is%20Valid
          (LocalVariables|SetRiderActor _returnvalue)
          ;[1 1 1] Reaction on Building Actors
          (bind _asbuilding_actor (Utilities|Casting|CastToBuildingActor (Variables|GetObject))
            (:then
              (if (> (Health|GetHealth _asbuilding_actor :!class /Script/FortniteGame.BuildingActor) 0)
                ;[1 1 1] Rumble + Cam Shake Building - STILL ALIVE
                (GameplayCue|ExecuteGameplayCue (LocalVariables|GetRiderActor) (Variables|Default|GetSprintChargeImpact_Default_CueTag))
                (else
                  ;[1 1 1] Rumble + Cam Shake Building - DESTROYED
                  (GameplayCue|ExecuteGameplayCue (LocalVariables|GetRiderActor) (Variables|Default|GetSprintChargeImpact_DestroyBuild_CueTag)))))
            (:CastFailed
              ;[1 1 1] Reaction on Pawns
              (bind _aspawn (Utilities|Casting|CastToPawn (Variables|GetObject))
                (:then
                  (if (Utilities|IsValidClass _gesprintimpactpawn)
                    (Fortnite|ApplyGameplayEffectsToActor _aspawn (Utilities|Array|MakeArray _gesprintimpactpawn) (Transformation|GetActorLocation (Components|GetOwner)) false 0 (Ability|GetAbilitySystemComponent (LocalVariables|GetRiderActor)))
                    ;[1 1 1] Rumble + Cam Shake Pawn
                    (GameplayCue|ExecuteGameplayCue (LocalVariables|GetRiderActor) (Variables|Default|GetSprintChargeImpact_Pawn_CueTag))))
                (:CastFailed
                  ;[1 1 1] Reaction on Vehicles
                  (bind _asfort_athena_vehicle (Utilities|Casting|CastToFortAthenaVehicle (Variables|GetObject)))
                  (if (Utilities|IsValidClass _gesprintimpactvehicle)
                    (Fortnite|ApplyGameplayEffectsToActor _asfort_athena_vehicle (Utilities|Array|MakeArray _gesprintimpactvehicle) (Transformation|GetActorLocation (Components|GetOwner)) false 0 (Ability|GetAbilitySystemComponent (LocalVariables|GetRiderActor)))
                    ;[1 1 1] Rumble + Cam Shake Vehicle
                    (GameplayCue|ExecuteGameplayCue (LocalVariables|GetRiderActor) (Variables|Default|GetSprintChargeImpact_Default_CueTag)))))))
          ;[1 1 1] Reaction VFX/SFX
          (GameplayCue|ExecuteGameplayCueOnActor(Burst) (Components|GetOwner) (Variables|Default|GetSprintImpactGameplayCueTag) (GameplayCue|MakeGameplayCueParametersfromHitResult (Variables|GetHitResult))))))))
