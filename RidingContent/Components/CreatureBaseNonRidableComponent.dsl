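; Creature component for non-ridable creatures, handling ride attempts with an optional kill response.
; Both the ToggleRidingAlternative and AlternateRidingEvent events check the KillOnRideAttempt flag, then cast the component's owner to FortPawn and force-kill it.
; GetRidingInfoFromTarget returns the jump-attach ground height minimum, its buffer and whether an alternate riding message is used.
; BeginPlay and Tick are empty.
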(event EventBeginPlay)

(event EventTick
  (params (DeltaSeconds :type Float%20%28single-precision%29)))

(event EventToggleRidingAlternative
  (if (Variables|Default|GetKillOnRideAttempt)
    (bind _asfort_pawn (Utilities|Casting|CastToFortPawn (Components|GetOwner)))
    (FortPawn|ForceKill :self _asfort_pawn)))

(event Custom|AlternateRidingEvent
  (if (Variables|Default|GetKillOnRideAttempt)
    (bind _asfort_pawn (Utilities|Casting|CastToFortPawn (Components|GetOwner)))
    (FortPawn|ForceKill :self _asfort_pawn)))

(fn GetRidingInfoFromTarget
  (return (Variables|RidingAttach|GetJump%20Attach%20Ground%20Height%20Min) (Variables|RidingAttach|GetJump%20Attach%20Ground%20Height%20Buffer) (Variables|RidingAttach|GetUses%20Alt%20Riding%20Message)))
