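; Base ridable component for Fortnite wildlife creatures, providing functions for mounting, taming, petting checks, rider capsule sizing and reversible collision overrides.
; Energy helpers drain stamina by sprint or movement state and recharge it without a rider.
; Jump fatigue counts closely chained jumps, applies a gameplay effect, then removes it on a post-landing timer.
; Jump exits require airtime; a velocity-projected sphere trace buffers them near ground, allowing conversion into jumps.
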(fn ApplyCollisionResponse
  (params (Channels :type Array%20of%20ECollisionChannel%20Enums :default "ECC_WorldStatic") (Response :type ECollisionResponse%20Enum :default "ECR_Ignore"))
  (bind _returnvalue (Components|GetOwner))
  (bind _returnvalue_2 (Actor|GetComponentbyClass _returnvalue "/Script/Engine.CapsuleComponent"))
  (Utilities|IsValid _returnvalue
    (:Is%20Valid
      (Utilities|IsValid _returnvalue_2
        (:Is%20Valid
          (for _array_element (Variables|GetChannels)
            (Utilities|Map|Add (Variables|Collision|GetOriginalCollisionResponseMap) _array_element (Collision|GetCollisionResponsetoChannel _returnvalue_2 _array_element))
            (Collision|SetCollisionResponsetoChannel _returnvalue_2 _array_element (Variables|GetResponse))))))))

(fn ApplyGEToSelf
  (params (GameplayEffectClass :type Gameplay%20Effect%20Class%20Reference))
  (bind _returnvalue (Utilities|IsValidClass GameplayEffectClass))
  (if _returnvalue
    (GameplayEffects|FortApplyGameplayEffectToSelf :self (Ability|GetFortAbilitySystemComponent (Components|GetOwner)) :GameplayEffectClass GameplayEffectClass :Level 0.0)
    (return _returnvalue)
    (else
      (return _returnvalue))))

(fn CanBePet
  (params (Rider :type Rider%20Component%20Object%20Reference))
  (if (|CallParentFunction Rider)
    (bind _ascontrolling_rider_component (Utilities|Casting|CastToControllingRiderComponent Rider)
      (:then
        (if (not (Class|ControllingRiderComponent|IsTryingtoMove _ascontrolling_rider_component 0.0))
          (bind _asfort_pawn (Utilities|Casting|CastToFortPawn (Components|GetOwner))
            (:then
              (Utilities|IsValid (Variables|/Script/Engine.Character|GetCharacterMovement _asfort_pawn)
                (:Is%20Valid
                  (if (not (or (WaterBody|IsSurfaceSwimming _asfort_pawn) (AI|Components|NavMovement|IsFalling (Variables|/Script/Engine.Character|GetCharacterMovement _asfort_pawn))))
                    (bind _asfort_player_pawn (Utilities|Casting|CastToFortPlayerPawn (Components|GetOwner Rider))
                      (:then
                        (if (not (GameplayTags|HasAnyMatchingGameplayTags _asfort_player_pawn (Variables|Default|GetBlockPetting_TagContainer)))
                          (return true)))
                      (:CastFailed
                        _))))
                (:Is%20Not%20Valid
                  _)))
            (:CastFailed
              _))))
      (:CastFailed
        _)))
  (return false))

(fn CanJumpExit
  (bind _returnvalue (Utilities|Time|GetTimeSeconds))
  (bind _jumpexitdurationthreshold (LocalVariables|GetJumpExitDurationThreshold))
  (bind _returnvalue_2 (ScalableFloat|GetValueatLevel _jumpexitdurationthreshold))
  (bind _startinairtimestamp (Variables|Jump|GetStartInAirTimestamp))
  (bind _returnvalue_3 (- _returnvalue _startinairtimestamp))
  (bind _npcpawnwildliferef (Variables|Default|GetNPCPawnWildlifeRef))
  (bind _charactermovement (Variables|/Script/Engine.Character|GetCharacterMovement _npcpawnwildliferef))
  (Utilities|IsValid _npcpawnwildliferef
    (:Is%20Valid
      (Utilities|IsValid _charactermovement
        (:Is%20Valid
          (if (AI|Components|NavMovement|IsFalling _charactermovement)
            (if (> _returnvalue_3 _returnvalue_2)
              (return true)
              (else
                (bind _returnvalue_4 (Utilities|Time|SetTimerbyFunctionName 0 "ReTryJumpExit" (- _returnvalue_2 _returnvalue_3)))
                (Variables|Jump|SetRetryJumpExitTimerHandle _returnvalue_4)
                (return false)))
            (else
              _)))
        (:Is%20Not%20Valid
          _)))
    (:Is%20Not%20Valid
      _))
  (return false))

(fn CanRiderPlayEmote
  (params (Rider :type Rider%20Component%20Object%20Reference))
  (return false))

(fn CommitEnergyCosts
  (params (RequireAllEnergy :type Boolean) (EnergyCost :type Float%20%28double-precision%29))
  (if (and (or (not RequireAllEnergy) (>= (Variables|EnergyTrackingInternal|GetcurrentEnergyValue) EnergyCost)) (> (Variables|EnergyTrackingInternal|GetcurrentEnergyValue) 0))
    (CallFunction|UseEnergyAmountServer EnergyCost)
    (return true)
    (else
      (return false))))

(fn DetermineRateOfChange
  (bind _npcpawnwildliferef (Variables|Default|GetNPCPawnWildlifeRef))
  (if (CallFunction|HasPrimaryRider)
    (if (Variables|EnergyTracking|GetinfiniteStaminaBuffEnabled)
      (return 0.0)
      (elif (CallFunction|IsSprinting)
        (bind _outxy (Fortnite|EvaluateCurveTableRow (Variables|EnergyTracking|GetEnergyRateOfDecaySprinting)))
        (return _outxy)
        (else
          (Utilities|IsValid _npcpawnwildliferef
            (:Is%20Valid
              (if (> (Math|Vector|VectorLength (Pawn|Components|CharacterMovement|GetLastUpdateVelocity (Variables|/Script/Engine.Character|GetCharacterMovement _npcpawnwildliferef))) (Variables|EnergyTrackingInternal|GetminMovementSpeedTreshold))
                (bind _outxy_2 (Fortnite|EvaluateCurveTableRow (Variables|EnergyTracking|GetEnergyRateOfDecayMoving)))
                (return _outxy_2)
                (else
                  _)))
            (:Is%20Not%20Valid
              _))
          (return (Variables|EnergyTracking|GetEnergyRateOfDecayStopped)))))
    (else
      (return (Variables|EnergyTracking|GetCurrentEnergyRateOfRecharge)))))

(fn ForceTaming
  (params (PlayerPawn :type Actor%20Object%20Reference) (CreaturePawn :type Actor%20Object%20Reference))
  (bind _returnvalue (Actor|GetComponentbyClass PlayerPawn "/Script/RidingCodeRuntime.RiderComponent"))
  (bind _ridingallowed (LocalVariables|GetRidingAllowed))
  (bind _returnvalue_2 (not _ridingallowed))
  (Utilities|IsValid _returnvalue
    (:Is%20Valid
      (LocalVariables|SetRidingAllowed (not (Class|RiderComponent|HasBlockingTag _returnvalue))))
    (:Is%20Not%20Valid
      _))
  (bind _returnvalue_3 (GameplayTags|HasAnyMatchingGameplayTags(Message) (Variables|GetCreaturePawn) (LocalVariables|GetCreativeAllowTamingTags))) ; Check if we only allow taming
  (if (or _returnvalue_3 _returnvalue_2)
    (bind _asfort_pawn (Utilities|Casting|CastToFortPawn (Variables|GetPlayerPawn)))
    (bind _asfort_pawn_2 (Utilities|Casting|CastToFortPawn (Variables|GetCreaturePawn)))
    (Convert|TriggerConversion _asfort_pawn _asfort_pawn_2 true))
  (return (or _returnvalue_3 _returnvalue_2)))

(fn GetRiderCapsuleSize
  (params (Rider :type Rider%20Component%20Object%20Reference))
  (bind _returnvalue (Components|GetOwner Rider))
  (bind (_outradius _outhalfheight) (Components|Capsule|GetUnscaledCapsuleSize (Actor|GetComponentbyClass _returnvalue "/Script/Engine.CapsuleComponent")))
  (Utilities|IsValid Rider
    (:Is%20Valid
      (Utilities|IsValid _returnvalue
        (:Is%20Valid
          (return _outradius _outhalfheight))
        (:Is%20Not%20Valid
          _)))
    (:Is%20Not%20Valid
      _))
  (return 0.0 0.0))

(fn GetRiderLocation
  (bind _donotuse_primaryrider (Variables|RidableComponent|GetDoNotUse_PrimaryRider))
  (Utilities|IsValid _donotuse_primaryrider
    (:Is%20Valid
      (return (Transformation|GetActorLocation (Components|GetOwner _donotuse_primaryrider))))
    (:Is%20Not%20Valid
      (return (Transformation|GetActorLocation (Components|GetOwner))))))

(fn GetRidingCapsuleSize
  (params (Rider :type Rider%20Component%20Object%20Reference) (OriginalRadius :type Float%20%28single-precision%29) (OriginalHalfHeight :type Float%20%28single-precision%29))
  (bind (_ridercapsuleradius _ridercapsulehh) (Default|GetRiderCapsuleSize Rider))
  (return (Math|Float|Max(Float) (+ OriginalRadius (Variables|Collision|GetCapsuleRadiusOffset)) _ridercapsuleradius) (Math|Float|Max(Float) (+ OriginalHalfHeight (Variables|Collision|GetCapsuleHalfHeightOffset)) _ridercapsulehh)))

(fn GetRidingEmoteCapsuleSize
  (params (Rider :type Rider%20Component%20Object%20Reference) (OriginalRadius :type Float%20%28single-precision%29) (OriginalHalfHeight :type Float%20%28single-precision%29))
  (bind (_outradius _outhalfheight) (CallFunction|GetRidingCapsuleSize Rider OriginalRadius OriginalHalfHeight))
  (return (+ _outradius (Variables|Collision|GetCapsuleRadiusOffset_Emote)) (+ _outhalfheight (Variables|Collision|GetCapsuleHalfHeightOffset_Emote))))

(fn GetRidingInfoFromTarget
  (return (Variables|RidingAttach|GetJumpAttach_GroundHeightMin) (Variables|RidingAttach|GetJumpAttach_GroundHeightBuffer) (Variables|RidingAttach|GetUseAltRidingMessage)))

(fn HandleEnemyFoleySoundLibrary
  (params (Rider :type Actor%20Component%20Object%20Reference) (bIsRiding :type Boolean))
  (bind _returnvalue (Components|GetOwner Rider))
  (bind _returnvalue_2 (WorldSubsystems|GetSoundLibrarySubsystem))
  (bind _ridingfoleysoundlibrary (Variables|Default|GetRidingFoleySoundLibrary))
  (if bIsRiding
    (bind _asfort_pawn (Utilities|Casting|CastToFortPawn _returnvalue))
    (switch Utilities|FlowControl|Switch|SwitchonEFortTeamAffiliation (Team|GetTeamAffiliationforActortoLocalPlayer _asfort_pawn)
      (:Hostile
        (SoundLibrary|AddLibrary _returnvalue_2 _returnvalue _ridingfoleysoundlibrary :!class /Script/SoundLibrary.SoundLibrarySubsystem)))
    (else
      (SoundLibrary|RemoveLibrary _returnvalue_2 _returnvalue _ridingfoleysoundlibrary :!class /Script/SoundLibrary.SoundLibrarySubsystem)))
  ;[1 1 1] Extend footstep relevancy distance for enemies riding
  (bind _npcpawnwildliferef (Variables|Default|GetNPCPawnWildlifeRef)
    (:then
      (Utilities|IsValid (Actor|GetComponentbyClass _npcpawnwildliferef "/Script/FortniteGame.FortFootstepComponent")
        (:Is%20Valid
          (Variables|/Script/FortniteGame.FortFootstepComponent|SetMaxFootstepDistance (Math|Float|SelectFloat 1e+04 2.5e+03 bIsRiding) (Actor|GetComponentbyClass _npcpawnwildliferef "/Script/FortniteGame.FortFootstepComponent")))))
    (:else)))

(fn IsSwimming
  (bind _asfort_pawn (Utilities|Casting|CastToFortPawn (Components|GetOwner))
    (:then
      (return (WaterBody|IsSurfaceSwimming _asfort_pawn)))
    (:CastFailed
      (return false))))

(fn JumpFatigue_CheckApplyFatigue
  ;[1 0.2077 0.2376] Jump Fatigue is activated & we have a valid Fatigue GE to apply
  (if (and (ScalableFloat|GetValueasBool (Variables|Jump|GetJumpFatigue_Enabled)) (Utilities|IsValidClass (Variables|Jump|GetJumpFatigueGameplayEffect)))
    (bind _asfort_aipawn (Utilities|Casting|CastToFortAIPawn (Components|GetOwner)))
    (if (>= (Variables|Jump|GetJumpFatigueJumpCount) (ScalableFloat|GetValueasInteger (Variables|Jump|GetJumpFatigue_ApplyFatigueMinJumpCount)))
      ;[0.6615 0.4828 1] Apply jump fatigue GE
      (bind _returnvalue (GameplayEffects|FortApplyGameplayEffectToSelf :self (Variables|/Script/FortniteGame.FortAIPawn|GetAIPawnAbilitySystemComponent _asfort_aipawn) :GameplayEffectClass (Variables|Jump|GetJumpFatigueGameplayEffect) :Level 0.0))
      (Variables|Jump|SetJumpFatigue_GEHandle _returnvalue)
      ;[1 1 1] Clear GE removal since we just reapplied a new GE
      (Utilities|Time|ClearandInvalidateTimerbyHandle (Variables|Jump|GetJumpFatigue_RemovalTimerHandle))
      ;[0.4297 0.9101 1] Optional Debug print
      (if (Variables|Jump|GetJumpFatigueDebugEnabled)
        (Development|PrintString "RideableJumpFatigue -> GE Applied" true true "(R=1.000000,G=0.000000,B=0.839051,A=1.000000)" 5.0)))))

(fn JumpFatigue_CountJumps
  ;[1 0.2077 0.2376] Jump Fatigue is enabled
  (if (ScalableFloat|GetValueasBool (Variables|Jump|GetJumpFatigue_Enabled))
    (Utilities|FlowControl|Sequence
      (:then_0
        ;[1 1 1] Since we have jumped: next time we land we timestamp. This was added to differentiate falling off a cliff from jumping.
        (Variables|Jump|SetJumpFatigue_ShouldTimeStampNextLanding true))
      (:then_1
        ;[1 0.2077 0.2376] Do we currently have jump fatigue applied?
        (bind _returnvalue (Ability|GameplayEffect|GetActiveGameplayEffectStackCount (Variables|Jump|GetJumpFatigue_GEHandle)))
        (if (or (> _returnvalue 0) (<= (- (Utilities|Time|GetGameTimeinSeconds) (Variables|Jump|GetJumpFatigue_LastLandingTimeStamp)) (ScalableFloat|GetValueatLevel (Variables|Jump|GetJumpFatigue_ResetCounterDelay))))
          (Math|Integer|IncrementInt (Variables|Jump|GetJumpFatigueJumpCount)
            (:%20%20
              _))
          (else
            (Variables|Jump|SetJumpFatigueJumpCount 1)))
        ;[0.4297 0.9101 1] Optional Debug Print
        (if (Variables|Jump|GetJumpFatigueDebugEnabled)
          (Development|PrintString (Utilities|String|ToString(Text) (Utilities|Text|FormatText "NSLOCTEXT(\"[7A7185181851B5D73310BD63578F29AD]\", \"D39620614E28758ED0F3BBB2702DAFDE\", \"RideableJumpFatigue JumpCount = {JumpCount}\")" (Variables|Jump|GetJumpFatigueJumpCount))) true true "(R=1.000000,G=0.000000,B=0.839051,A=1.000000)" 5.0))))))

(fn JumpFatigue_Landed
  ;[1 0.2077 0.2376] Jump Fatigue is activated & we have a valid Fatigue GE to apply
  (if (and (ScalableFloat|GetValueasBool (Variables|Jump|GetJumpFatigue_Enabled)) (Utilities|IsValidClass (Variables|Jump|GetJumpFatigueGameplayEffect)))
    ;[1 0.2077 0.2376] Do we currently have applied jump fatigue GE?
    (bind _returnvalue (Ability|GameplayEffect|GetActiveGameplayEffectStackCount (Variables|Jump|GetJumpFatigue_GEHandle)))
    (if (> _returnvalue 0)
      ;[1 1 1] Call "RemoveFatigue" when timer will expire.
      (bind _returnvalue_2 (Utilities|Time|SetTimerbyFunctionName 0 "JumpFatigue_RemoveFatigue" (ScalableFloat|GetValueatLevel (Variables|Jump|GetJumpFatigue_GeDurationPostLanding))))
      (Variables|Jump|SetJumpFatigue_RemovalTimerHandle _returnvalue_2))))

(fn JumpFatigue_LandTimeStamp
  ;[0.6244 1 0.6475] Timestamp for "JumpFatigue_CountJumps" function to know if we are making consecutive jumps
  (if (Variables|Jump|GetJumpFatigue_ShouldTimeStampNextLanding)
    (Variables|Jump|SetJumpFatigue_LastLandingTimeStamp (Utilities|Time|GetGameTimeinSeconds))
    (Variables|Jump|SetJumpFatigue_ShouldTimeStampNextLanding false)))

(fn JumpFatigue_RemoveFatigue
  (bind _asfort_aipawn (Utilities|Casting|CastToFortAIPawn (Components|GetOwner)))
  ;[0.6615 0.4828 1] Remove all jump fatigue GE stacks and clear variables
  (GameplayEffects|RemoveActiveGameplayEffect (Variables|/Script/FortniteGame.FortAIPawn|GetAIPawnAbilitySystemComponent _asfort_aipawn) (Variables|Jump|GetJumpFatigue_GEHandle) -1)
  (Utilities|Time|ClearandInvalidateTimerbyHandle (Variables|Jump|GetJumpFatigue_RemovalTimerHandle))
  ;[0.4297 0.9101 1] Optional Debug Print
  (if (Variables|Jump|GetJumpFatigueDebugEnabled)
    (Development|PrintString "RideableJumpFatigue -> GE Removed" true true "(R=1.000000,G=0.000000,B=0.839051,A=1.000000)" 5.0)))

(fn KickRiderOffSelf
  (bind _donotuse_primaryrider (Variables|RidableComponent|GetDoNotUse_PrimaryRider))
  (Utilities|IsValid _donotuse_primaryrider
    (:Is%20Valid
      (CallFunction|ToggleRiding (Components|GetOwner _donotuse_primaryrider) (Components|GetOwner)))))

(fn OnMovementModeChanged
  (params (Character :type Character%20Object%20Reference) (PrevMovementMode :type EMovementMode%20Enum :default "MOVE_None") (PreviousCustomMode :type Byte))
  (bind _asfort_pawn (Utilities|Casting|CastToFortPawn Character))
  (Utilities|FlowControl|Sequence
    (:then_0
      ;[1 1 1] If we start swimming > stop sprint
      (if (WaterBody|IsSurfaceSwimming _asfort_pawn)
        (CallFunction|StopSprinting)
        (Utilities|SwitchHasAuthority
          (:Authority
            (CallFunction|JumpFatigueLandTimeStamp)
            (CallFunction|JumpFatigueLanded))
          (:Remote
            _))
        (CallFunction|TryConvertBufferedJumpExittoJump)))
    (:then_1
      ;[1 1 1] Set CanJumpExit Timestamp
      (if (AI|Components|NavMovement|IsFalling (Variables|/Script/Engine.Character|GetCharacterMovement _asfort_pawn))
        (Variables|Jump|SetStartInAirTimestamp (Utilities|Time|GetTimeSeconds))
        (else
          (Utilities|Time|ClearandInvalidateTimerbyHandle (Variables|Jump|GetRetryJumpExitTimerHandle))
          (Variables|Jump|SetStartInAirTimestamp 0.0))))))

(fn OnRep_currentEnergyValue
  (Default|CallRidableUIUpdateEnergy (Variables|EnergyTrackingInternal|GetcurrentEnergyValue) (> (Variables|EnergyTracking|GetCurrentEnergyRateOfRecharge) 0)))

(fn OnRep_IdleAnim_Add)

(fn PropagateImmunityFallDamage
  (bind _playerridercomponentref (Variables|Default|GetPlayerRiderComponentRef))
  (bind _returnvalue (Components|GetOwner _playerridercomponentref))
  (if (ScalableFloat|GetValueasBool (LocalVariables|GetCanProbagateImmunityFallDanage))
    (Utilities|IsValid _playerridercomponentref
      (:Is%20Valid
        (Utilities|IsValid _returnvalue
          (:Is%20Valid
            (LocalVariables|SetRiderActor _returnvalue)
            (bind _asfort_pawn (Utilities|Casting|CastToFortPawn (Components|GetOwner)))
            (switch Utilities|FlowControl|Switch|SwitchonEMovementMode (Variables|/Script/Engine.CharacterMovementComponent|GetMovementMode (Variables|/Script/Engine.Character|GetCharacterMovement _asfort_pawn))
              (:MOVE_Falling
                (if (or (Variables|/Script/FortniteGame.FortPawn|GetbIgnoreNextFallingDamage _asfort_pawn) (GameplayTags|HasMatchingGameplayTag _asfort_pawn "(TagName=\"Gameplay.Status.ImmuneFallingDamage\")"))
                  (Fortnite|ApplyGameplayEffectsToActor (LocalVariables|GetRiderActor) (LocalVariables|GetGE_ImmunityFallDamage) (Transformation|GetActorLocation (LocalVariables|GetRiderActor))))))))))))

(fn RemoveGEByTag
  (params (Tags :type Gameplay%20Tag%20Container%20Structure))
  (GameplayEffects|RemoveActiveEffectswithGrantedTags (Ability|GetFortAbilitySystemComponent (Components|GetOwner)) Tags))

(fn RestoreCollisionResponses
  (bind _originalcollisionresponsemap (Variables|Collision|GetOriginalCollisionResponseMap))
  (bind _keys (Utilities|Map|Keys (Variables|Collision|GetOriginalCollisionResponseMap)))
  (for _array_element _keys
    (CallFunction|ApplyCollisionResponse (Utilities|Array|MakeArray _array_element) (Utilities|Map|Find _originalcollisionresponsemap _array_element)))
  (Utilities|Map|Clear _originalcollisionresponsemap))

(fn SetInfiniteStaminaBuff
  (params (Enabled :type Boolean))
  (bind _output_get (Variables|EnergyTracking|SetinfiniteStaminaBuffEnabled Enabled))
  (if _output_get
    (CallFunction|OnEattoRefuel -1.0)
    (GameplayCue|AddGameplayCueOnActor(Looping) (Components|GetOwner) (Variables|Default|GetInfiniteStaminaEffect_GameplayCueTag) (Ability|GameplayCue|MakeGameplayCueParameters 0.0))
    (else
      (GameplayCue|RemoveGameplayCueOnActor(Looping) (Components|GetOwner) (Variables|Default|GetInfiniteStaminaEffect_GameplayCueTag) (Ability|GameplayCue|MakeGameplayCueParameters 0.0)))))

(fn SetWildlifeHealth
  (bind _npcpawnwildliferef (Variables|Default|GetNPCPawnWildlifeRef))
  (bind _returnvalue (FortPawn|GetMaxHealth _npcpawnwildliferef))
  (bind _returnvalue_2 (FortPawn|GetHealth _npcpawnwildliferef))
  (bind _returnvalue_3 (/ _returnvalue_2 _returnvalue))
  (bind _playerridercomponentref (Variables|Default|GetPlayerRiderComponentRef))
  (Utilities|IsValid _npcpawnwildliferef
    (:Is%20Valid
      (Utilities|IsValid _playerridercomponentref
        (:Is%20Valid
          (Class|PlayerRiderComponent|SetTargetHealth _playerridercomponentref _returnvalue_3 :!class /RidingContent/Components/PlayerRiderComponent.PlayerRiderComponent_C)
          (Development|PrintText (Utilities|Text|FormatText "NSLOCTEXT(\"[7A7185181851B5D73310BD63578F29AD]\", \"6F5604A440776E6B1230A2899A24CF03\", \"CreatureBaseRidableComponent-SetWildlifeHealth value {0}\")" _returnvalue_3)))))))

(fn ShouldConvertJumpExittoJump
  ;[1 1 1] Convert jump exit to jump if within buffer time
  (if (<= (- (Utilities|Time|GetTimeSeconds) (Variables|Jump|GetJumpExitBufferStartTimeStamp)) (ScalableFloat|GetValueatLevel (Variables|Jump|GetJumpExitBufferTime)))
    (Utilities|Time|ClearandInvalidateTimerbyHandle (Variables|Jump|GetJumpExitBufferTimerHandle))
    (Variables|Jump|SetJumpExitBufferStartTimeStamp 0.0)
    (return true)
    (else
      (return false))))

(fn ShouldBufferJumpExitToRetryJump
  (bind _npcpawnwildliferef (Variables|Default|GetNPCPawnWildlifeRef))
  (bind _charactermovement (Variables|/Script/Engine.Character|GetCharacterMovement _npcpawnwildliferef))
  (bind _velocity (Variables|/Script/Engine.MovementComponent|GetVelocity _charactermovement))
  (if (> (ScalableFloat|GetValueatLevel (Variables|Jump|GetJumpExitBufferTime)) 0)
    (Utilities|IsValid _npcpawnwildliferef
      (:Is%20Valid
        (Utilities|IsValid _charactermovement
          (:Is%20Valid
            (if (< (.z _velocity) 0)
              (bind (_outhit _returnvalue) (Collision|SphereTraceByChannel :Start (Transformation|GetActorLocation (Variables|Default|GetNPCPawnWildlifeRef)) :End (+ (* _velocity (ScalableFloat|GetValueatLevel (Variables|Jump|GetJumpExitBufferTime))) (Transformation|GetActorLocation (Variables|Default|GetNPCPawnWildlifeRef))) :Radius (Variables|/Script/Engine.CapsuleComponent|GetCapsuleRadius (Variables|/Script/Engine.Character|GetCapsuleComponent (Variables|Default|GetNPCPawnWildlifeRef))) :TraceChannel "TraceTypeQuery10" :bTraceComplex true :DrawDebugType "None" :bIgnoreSelf true :TraceColor "(R=0.000000,G=1.000000,B=0.153125,A=1.000000)" :TraceHitColor "(R=1.000000,G=0.013386,B=0.000000,A=1.000000)" :DrawTime 1e+01))
              (if _returnvalue
                (Variables|Jump|SetJumpExitBufferStartTimeStamp (Utilities|Time|GetTimeSeconds))
                (bind _returnvalue_2 (Utilities|Time|SetTimerbyFunctionName 0 "ExecuteBufferedJumpExit" (ScalableFloat|GetValueatLevel (Variables|Jump|GetJumpExitBufferTime))))
                (Variables|Jump|SetJumpExitBufferTimerHandle _returnvalue_2)
                (return true)
                (else
                  _))))
          (:Is%20Not%20Valid
            _)))
      (:Is%20Not%20Valid
        _)))
  (return false))

(fn ShouldCameraFocusOnRidable
  (params (Rider :type Rider%20Component%20Object%20Reference))
  (return true))

(fn ShouldSetAsViewTarget
  (params (Rider :type Rider%20Component%20Object%20Reference))
  (return false))

(fn StopSprinting
  (bind _playerridercomponentref (Variables|Default|GetPlayerRiderComponentRef))
  (bind _returnvalue (Components|GetOwner _playerridercomponentref))
  (bind _returnvalue_2 (Ability|GetAbilitySystemComponent _returnvalue))
  ;[1 1 1] Stop Sprinting GE on Creature
  (GameplayEffects|RemoveActiveGameplayEffectbySourceEffect (Ability|GetFortAbilitySystemComponent (Components|GetOwner)) (Variables|Default|GetGE_Creature_Sprint) 0 -1)
  ;[1 1 1] Stop Sprinting On Player
  (Utilities|IsValid _playerridercomponentref
    (:Is%20Valid
      (Utilities|IsValid _returnvalue
        (:Is%20Valid
          (Utilities|IsValid _returnvalue_2
            (:Is%20Valid
              (GameplayEffects|RemoveActiveGameplayEffectbySourceEffect _returnvalue_2 (Variables|Default|GetGameplay%20Effect) 0 -1)
              (Variables|Riding|SetbIsSprinting false)
              (Networking|ForceNetUpdate (Components|GetOwner)))))))))

(fn TestThreshold
  (params (isDecaying :type Boolean) (currentValue :type Float%20%28double-precision%29) (testTreshold :type Float%20%28double-precision%29) (previousState :type Boolean))
  (if isDecaying
    (return (and (< currentValue testTreshold) (not previousState)))
    (else
      (return (and (>= currentValue testTreshold) previousState)))))

(fn ToggleRiding
  (params (PlayerPawn :type Actor%20Object%20Reference) (Target :type Actor%20Object%20Reference) (ShouldRide :type Boolean))
  (bind _returnvalue (Actor|GetComponentbyClass PlayerPawn "/Script/RidingCodeRuntime.RiderComponent"))
  (bind _returnvalue_2 (GameplayTags|HasAnyMatchingGameplayTags(Message) Target (Variables|Default|GetT_CannotRideBucket))) ; No riding if held (like a chicken) or energy is depleted
  (if (not _returnvalue_2)
    (bind _returnvalue_3 (GameplayTags|HasAnyMatchingGameplayTags(Message) PlayerPawn (Variables|Default|GetCannot%20Ride%20Rider%20Bucket)))
    (if (not _returnvalue_3)
      (Utilities|IsValid _returnvalue
        (:Is%20Valid
          (bind _ascontrolling_rider_component (Utilities|Casting|CastToControllingRiderComponent _returnvalue)
            (:then
              (Variables|/Script/RidingCodeRuntime.ControllingRiderComponent|SetbCanControlRidable true _ascontrolling_rider_component))
            (:CastFailed
              _))
          (if ShouldRide
            (bind _returnvalue_4 (Riding|StartRidingActor _returnvalue Target 0))
            (return _returnvalue_4)
            (else
              _))))))
  (bind _returnvalue_5 (Riding|StopRidingActor _returnvalue true true false true))
  (return _returnvalue_5))

(fn UpdateAnimInputBool
  (params (FortPawn :type Fort%20Pawn%20Object%20Reference) (isBeingRidden :type Boolean))
  (bind _returnvalue (Class|FortAnimInputEventBool|CreateAnimInputEventBool "aie.riding.IsBeingRidden" isBeingRidden))
  (Animation|TriggerAnimInputEvent FortPawn _returnvalue))

(fn UpdateLootAnimalGlow
  (params (Scale :type Float%20%28double-precision%29))
  (bind _returnvalue (Components|GetOwner))
  (bind _returnvalue_2 (Actor|GetComponentsbyTag _returnvalue "/Script/Niagara.NiagaraComponent" "Irwin.Component.LootGlow"))
  (bind _output (Utilities|Array|Get(acopy) _returnvalue_2))
  (bind _returnvalue_3 (GameplayTags|HasAnyMatchingGameplayTags(Message) _returnvalue (GameplayTags|MakeGameplayTagContainerfromTag "(TagName=\"Pawn.Athena.NPC.Wildlife.Variant.Loot\")")))
  (if _returnvalue_3
    (Effects|Components|ParticleSystem|SetFloatParameter _output "LootEffectScale" Scale)
    (Niagara|ReinitializeSystem _output)))

(fn UpdateValue
  (params (targetValue :type Float%20%28double-precision%29) (rateOfChange :type Float%20%28double-precision%29) (deltaTime :type Float%20%28double-precision%29))
  (bind _returnvalue (* rateOfChange deltaTime))
  (bind _returnvalue_2 (+ targetValue _returnvalue))
  (if (>= _returnvalue_2 1.0)
    (return 1.0 false)
    (elif (<= _returnvalue_2 0)
      (return 0.0 true)
      (else
        (return _returnvalue_2 false)))))
