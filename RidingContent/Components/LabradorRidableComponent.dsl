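; Ridable component for the Labrador wildlife NPC that converts riding state into map, tracking and camera effects.
; Mounting deactivates the map marker, optionally grants a tracking ability set, and adds a camera modifier toggled by alternating gameplay-tag waits on IsTargeting.
; Dismounting triggers "Dismiss" unconversion through a server event, removing tracking and camera modifier; the marker stays off by design.
; Toppling kicks the rider off and blocks riding.
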(event EventBeginPlay
  (|CallParentFunction)
  (bind _asbp_aipawn_labrador_ridable (Utilities|Casting|CastToBP_AIPawn_Labrador_Ridable (Variables|Default|GetNPCPawnWildlifeRef)))
  (bind _output_get (Variables|Default|SetLabradorPawn _asbp_aipawn_labrador_ridable))
  (Default|BindEventtoToppledDispatcher _output_get (EventDispatchers|CreateEvent :!event EventToppled)))

(event EventHandleRiderStartedRiding
  (params (Rider :type Rider%20Component%20Object%20Reference) (SlotIndex :type Integer))
  (bind _returnvalue (Components|GetOwner Rider))
  ; Unconvert when player dismounts
  (|CallParentFunction Rider 0)
  (EventDispatchers|BindEventtoOnStoppedRidingSingularDeprecated Rider (EventDispatchers|CreateEvent :!event UnconvertLabrador))
  (Components|Activation|Deactivate (Variables|/Labrador/Pawn/BP_AIPawn_Labrador_Ridable.BP_AIPawn_Labrador_Ridable_C|GetNPCSpecialActorComponent (Variables|Default|GetLabradorPawn))) ; Remove map mark while riding
  (Utilities|FlowControl|Sequence
    (:then_0
      (if (ScalableFloat|GetValueasBool (Variables|Default|GetHF_TrackingEnabled))
        (bind _asfort_pawn (Utilities|Casting|CastToFortPawn (Components|GetOwner Rider)))
        (bind _returnvalue_2 (Ability|EquipFortAbilitySet _asfort_pawn "/Labrador/Gameplay/Abilities/Tracking/AS_Labrador_GrantTracking.AS_Labrador_GrantTracking")) ; Apply medallion tracking
        (Variables|Default|SetTrackingAbilitySet _returnvalue_2)))
    (:then_1
      (bind _returnvalue_3 (Fortnite|Camera|AddCameraModifierforOwningPlayerController "/RidingContent/Gameplay/Camera/CameraModifier_RidingCamera_Labrador.CameraModifier_RidingCamera_Labrador_C" _returnvalue))
      (Variables|Default|SetTargetingCamMod _returnvalue_3)
      (CallFunction|WaitforTargetingStart _returnvalue))))

(event EventHandleRiderStoppedRiding
  (params (Rider :type Rider%20Component%20Object%20Reference) (SlotIndex :type Integer))
  ; Unconvert when player dismounts
  (|CallParentFunction Rider 0)
  (EventDispatchers|UnbindEventfromOnStoppedRidingSingularDeprecated Rider (EventDispatchers|CreateEvent :!event UnconvertLabrador)))

(event Custom|WaitForTargetingStart
  (params (TargetActor :type Actor%20Object%20Reference))
  ; Unconvert when player dismounts
  (bind _asynctaskproxy (Ability|Tasks|WaitGameplayTagAddToActor TargetActor "(TagName=\"Gameplay.Status.IsTargeting\")" true)
    (:then
      (Variables|Default|SetTask_Targeting _asynctaskproxy))
    (:Added
      (bind _donotuse_primaryrider (Variables|RidableComponent|GetDoNotUse_PrimaryRider)
        (:then
          (CameraModifier|EnableModifier (Variables|Default|GetTargetingCamMod))
          (CallFunction|WaitforTargetingEnd (Components|GetOwner _donotuse_primaryrider)))
        (:else)))))

(event Custom|WaitForTargetingEnd
  (params (TargetActor :type Actor%20Object%20Reference))
  ; Unconvert when player dismounts
  (bind _asynctaskproxy (Ability|Tasks|WaitGameplayTagRemoveFromActor TargetActor "(TagName=\"Gameplay.Status.IsTargeting\")" true)
    (:then
      (Variables|Default|SetTask_EndTargeting _asynctaskproxy))
    (:Removed
      (bind _donotuse_primaryrider (Variables|RidableComponent|GetDoNotUse_PrimaryRider)
        (:then
          (CameraModifier|DisableModifier (Variables|Default|GetTargetingCamMod) true)
          (CallFunction|WaitforTargetingStart (Components|GetOwner _donotuse_primaryrider)))
        (:else)))))

(event Custom|Server_TriggerUnconversion
  (params (Target :type Actor%20Component%20Object%20Reference))
  ; Unconvert when player dismounts
  (bind _asfort_pawn (Utilities|Casting|CastToFortPawn (Components|GetOwner Target)))
  (Convert|TriggerUnconversion _asfort_pawn (Variables|Default|GetLabradorPawn) "Dismiss")
  (Ability|UnequipFortAbilitySet (Variables|Default|GetTrackingAbilitySet)) ; End medallion tracking
  (if false
    (Components|Activation|Activate (Variables|/Labrador/Pawn/BP_AIPawn_Labrador_Ridable.BP_AIPawn_Labrador_Ridable_C|GetNPCSpecialActorComponent (Variables|Default|GetLabradorPawn)) false) ; Reapply map mark, disabled by design
  ))

(fn CanBeRidden
  (params (Rider :type Rider%20Component%20Object%20Reference) (SlotIndex :type Integer))
  (return (and (|CallParentFunction Rider) (not (Variables|/Labrador/Pawn/BP_AIPawn_Labrador.BP_AIPawn_Labrador_C|GetIsToppled (Variables|Default|GetLabradorPawn))))))

(fn EventToppled
  (params (isToppled :type Boolean))
  (if isToppled
    (CallFunction|KickRiderOffSelf)))

(fn ToggleTargetingCameraMode)

(fn UnconvertLabrador
  (params (Rider :type Rider%20Component%20Object%20Reference) (Ridable :type Ridable%20Component%20Object%20Reference))
  (Utilities|IsValid Rider
    (:Is%20Valid
      (CallFunction|ServerTriggerUnconversion Rider)
      (Game|Player|RemoveCameraModifier (Variables|/Script/Engine.PlayerController|GetPlayerCameraManager (Fortnite|GetFortPlayerControllerfromActor (Components|GetOwner Rider))) (Variables|Default|GetTargetingCamMod)))))
