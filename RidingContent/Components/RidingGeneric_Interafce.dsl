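; Generic riding interface in the riding content's components, declaring functions related to riding targets and riding mode.
; GetRidingInfoFromTarget returns two numeric values and a boolean, all left at zero or false.
; ToggleRidingAlternative takes no visible inputs, returns nothing and has no body, so any behavior comes from implementers.
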(fn GetRidingInfoFromTarget
  (return 0.0 0.0 false))

(fn ToggleRidingAlternative)
